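# Physics asset for the Street Racer White backpack cosmetic, providing bodies for its sword and dangling tassels.
# Six capsule bodies cover the backpack root, sword, two-segment tassel and two wrap tassels.
# The sword is constrained to the root, the tassel segments chain from the sword, and both wrap tassels attach to it separately.
# Every joint locks swing1, and the sword joint also locks twist, leaving it a single swing axis.
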PhysicsAsset /BRCosmetics/Accessories/FORT_Backpacks/F_MED_Street_Racer_White_Backpack/Meshes/SK_F_MED_Street_Racer_White_Physics
bodies {
 backpack_root [{shapeType=Capsule, length=16.25, radius=9.5, center=<1.7582 0.2669 -0.241>, rotation=<rot -1.1961 -90 0.3837>}]
 dyn_sword [{shapeType=Capsule, length=42.0683, radius=2.3107, center=<1.9177 -2.0654 0.1475>, rotation=<rot 9.2832 -90.9225 36.0425>}]
 dyn_tassle_01 [{shapeType=Capsule, length=8.25, radius=2.255, center=<-0.323 -0.0782 -4.6631>, rotation=<rot 0.5606 -90 2.5294>}]
 dyn_tassle_02 [{shapeType=Capsule, length=8.3391, radius=2.3321, center=<-0.4092 -0.1698 -4.453>, rotation=<rot 1.6046 -90 2.6698>}]
 dyn_wrap_tassleA [{shapeType=Capsule, length=4.6516, radius=1.3709, center=<0 0.0002 -2.6909>}]
 dyn_wrap_tassleB [{shapeType=Capsule, length=4.5265, radius=1.2664, center=<0 0.0002 -2.6909>}]
}
constraints [
 {
  bone1Name=dyn_sword
  bone2Name=backpack_root
  swing1LimitDegrees=45
  swing1Motion=Locked
  swing2LimitDegrees=15
  twistLimitDegrees=45
  twistMotion=Locked
 }
 {bone1Name=dyn_tassle_01, bone2Name=dyn_sword, swing1LimitDegrees=45, swing1Motion=Locked, swing2LimitDegrees=45, twistLimitDegrees=45}
 {bone1Name=dyn_tassle_02, bone2Name=dyn_tassle_01, swing1LimitDegrees=45, swing1Motion=Locked, swing2LimitDegrees=45, twistLimitDegrees=45}
 {bone1Name=dyn_wrap_tassleB, bone2Name=dyn_sword, swing1LimitDegrees=45, swing1Motion=Locked, swing2LimitDegrees=30, twistLimitDegrees=45}
 {bone1Name=dyn_wrap_tassleA, bone2Name=dyn_sword, swing1LimitDegrees=45, swing1Motion=Locked, swing2LimitDegrees=30, twistLimitDegrees=45}
]
